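# Editor utility widget for tech art that lists the tag key/value pairs of a single asset in a non-selectable list.
# setCurrentAsset stores the asset and calls refreshList, which clears the list, builds AssetData and reads its tag map.
# Keys and values pair by loop index, skipping the FiBData tag.
# addKeyValue wraps each pair in a TA_GenericListItemData object and adds it to TA_Generic01_List, which uses MetadataViewer_ListItem rows.
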EditorUtilityWidgetBlueprint /Game/Blueprints/Blutilities/ForTechArt/AssetTools/AssetMetadataManager/ViewAssetData_EWidget
graphs [
 EventGraph {nodes [
   Event Construct()
   > Parent: Construct()
  ]}
 addKeyValue {nodes [
   n1 = addKeyValue!() -> (then, key:String, value:String) {is=title K2Node_FunctionEntry}
   > n2 = Construct Object from Class(Class=/Game/Blueprints/Blutilities/ForTechArt/Data/TA_GenericListItemData.TA_GenericListItemData_C, self=Self-Reference()) -> (ReturnValue:TA Generic List Item Data Object Reference)
   > Set strings!(
    execute
    strings:Array of Strings=Make Array!("[0]":String=@n1, "[1]":String=@n1.value) -> (Array:Array of Strings)
    self:TA Generic List Item Data Object Reference=@n2
   ) -> (then, Output_Get:Array of Strings)
   > addItem!(
    execute
    self:TA Generic 01 List Object Reference=Get TA_Generic01_List() -> (TA_Generic01_List:TA Generic 01 List Object Reference)
    dataObject:Object Reference=@n2
   ) -> (then, newItem:User Widget Object Reference, newData:Object Reference)
  ]}
 refreshList {nodes [
   refreshList() {is=title K2Node_FunctionEntry}
   > clearItems(self:TA Generic 01 List Object Reference=Get TA_Generic01_List())
   > Is Valid(InputObject=Get currentAsset() -> (currentAsset:Object Reference))
   n1 = CreateAssetData(InAsset=Get currentAsset(), bAllowBlueprintClass=false)
   > n2 = GetTagValues!(execute, AssetPath:String="To String (Name)"(InName=Break Asset Data(AssetData=@n1))) -> (then, ReturnValue:Map of Names to Strings)
   > n3 = Keys!(execute, TargetMap:Map of Names to Strings=@n2) -> (then, Keys:Array of Names)
   > n4 = For Each Loop(Array:Array of Names=@n3) -> (Array Element:Name)
   > n5 = Values!(execute, TargetMap:Map of Names to Strings=@n2) -> (then, Values:Array of Strings)
   > Branch(Condition="Not Equal (Name)"!(A:Name=@n4, B:Name=FiBData) -> (ReturnValue:bool) {is=title K2Node_CallFunction})
   > addKeyValue!(execute, self, key:String="To String (Name)"(InName=@n4), value:String="Get (a copy)"(Array=@n5, Dimension 1=@n4.Array Index)) -> (then) {is=title K2Node_CallFunction}
  ]}
 setCurrentAsset {nodes [
   n1 = setCurrentAsset!() -> (then, currentAsset:Object Reference) {is=title K2Node_FunctionEntry}
   > Set currentAsset(currentAsset:Object Reference=@n1) -> (Output_Get:Object Reference)
   > refreshList() {is=title K2Node_CallFunction}
  ]}
]
variables [currentAsset {type=Object, default=null}]
widgets {
 info {rootWidgetClass=&CanvasPanel, widgetCount=3}
 widgets [
  CanvasPanel_0 {widgetClassPath=&CanvasPanel, widget=CanvasPanel_0}
  Border_0 {
   widgetClassPath=&Border
   parent=CanvasPanel_0
   widget=Border_0
   properties {brushColor=<rgba 0.0296 0.0296 0.0296 1>, padding=<ltrb 5 5 5 5>}
   slot {class=CanvasPanelSlot, properties {layoutData {alignment=<0 0>, anchors {maximum=<1 1>, minimum=<0 0>}, offsets=<ltrb 10 10 10 10>}}}
  }
  TA_Generic01_List {
   widgetClassPath=/Game/Blueprints/Blutilities/ForTechArt/Widgets/List/TA_Generic01_List
   bIsVariable=true
   parent=Border_0
   widget=TA_Generic01_List
   properties {
    allowMultiSelect=false
    enableSelection=false
    listItemClass=&/MetadataViewer_ListItem.MetadataViewer_ListItem_C
   }
   slot {class=BorderSlot, properties {padding=<ltrb 5 5 5 5>}}
  }
 ]
}
