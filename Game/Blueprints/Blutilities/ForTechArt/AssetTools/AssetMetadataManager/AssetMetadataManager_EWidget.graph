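# Editor utility widget for tech art that hosts Metadata Editor and Tag Viewer tabs for a single asset.
# setCurrentAsset stores the asset, refreshes the header and passes it to each tab's child widget by casting, repeated when the tabs initialize.
# The header shows the asset name, or a red right-click hint with the find button hidden when none is valid.
# The find button syncs the Content Browser to the asset.
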EditorUtilityWidgetBlueprint /Game/Blueprints/Blutilities/ForTechArt/AssetTools/AssetMetadataManager/AssetMetadataManager_EWidget
graphs [
 EventGraph {nodes [
   "On Clicked (btnFindAsset)"!() -> (OutputDelegate, then, widget:User Widget Object Reference, event:Pointer Event Structure) {is=title K2Node_ComponentBoundEvent}
   > Is Valid(InputObject=Get currentAsset() -> (currentAsset:Object Reference))
   n1 = CreateAssetData!(InAsset=Get currentAsset(), bAllowBlueprintClass=false) -> (
    ReturnValue_PackageName:Name
    ReturnValue_PackagePath:Name
    ReturnValue_AssetName:Name
    ReturnValue_AssetClassPath:Top Level Asset Path Structure
   )
   > SyncBrowserToObjects(AssetPaths=Make Array("[0]":String="To String (Name)"(InName=@n1)) -> (Array:Array of Strings))
   "On Tabs Initialized (Generic01_TabContainer)"!() -> (OutputDelegate, then, TabContainer:Base Tab Container Widget Object Reference) {is=title K2Node_ComponentBoundEvent}
   > setCurrentAssetOnChildren()
   Event Construct()
   > Parent: Construct()
   > validateAssetText()
   n2 = Event Tick()
   > Parent: Tick!(execute, MyGeometry:Geometry Structure=@n2.MyGeometry, InDeltaTime:float=@n2.InDeltaTime) -> (then) {is=title K2Node_CallParentFunction}
  ]}
 setCurrentAsset {nodes [
   n1 = setCurrentAsset!() -> (then, currentAsset:Object Reference) {is=title K2Node_FunctionEntry}
   > Set currentAsset(currentAsset:Object Reference=@n1) -> (Output_Get:Object Reference)
   > validateAssetText()
   > setCurrentAssetOnChildren()
  ]}
 setCurrentAssetOnChildren {nodes [
   setCurrentAssetOnChildren() {is=title K2Node_FunctionEntry}
   > n1 = For Each Loop(Array:Array of TCW Tab Data Struct Structures=getCurrentTabData!(self:Base Tab Container Widget Object Reference=Get Generic01_TabContainer() -> (Generic01_TabContainer:Generic 01 Tab Container Object Reference)) -> (currentTabData:Array of TCW Tab Data Struct Structures)) -> (Array Element:TCW Tab Data Struct Structure)
   n2 = Break TCW Tab Data Struct!(TCW_TabData_Struct:"TCW Tab Data Struct Structure (by ref)"=@n1) -> (
    tabName_2_#1:String
    tabIndex_5_#2:int
    tabWidgetClass_11_#3:User Widget Class Reference
    tabContentWidget_14_#4:User Widget Object Reference
    tabButtonWidget_17_#5:User Widget Object Reference
    tabButtonStyle_20_#6:TCW Button Visuals Struct Structure
   ) {is=title K2Node_BreakStruct}
   > Is Valid(InputObject=@n2.tabContentWidget_14_#4)
   > n3 = Cast To ViewAssetData_EWidget!(execute, Object=@n2.tabContentWidget_14_#4) -> (then, CastFailed, AsView Asset Data EWidget:View Asset Data EWidget Object Reference) {is=title K2Node_DynamicCast}
   > setCurrentAsset!(execute, self:View Asset Data EWidget Object Reference=@n3, currentAsset:Object Reference=Get currentAsset()) -> (then) {is=title K2Node_CallFunction}
   n4 = Cast To EditAssetMetadata_EWidget!(execute=@n3.CastFailed, Object=@n2.tabContentWidget_14_#4) -> (then, CastFailed, AsEdit Asset Metadata EWidget:Edit Asset Metadata EWidget Object Reference) {is=title K2Node_DynamicCast}
   > setCurrentAsset(self:Edit Asset Metadata EWidget Object Reference=@n4, currentAsset=Get currentAsset())
  ]}
 validateAssetText {nodes [
   validateAssetText() {is=title K2Node_FunctionEntry}
   > n1 = Is Valid(InputObject=Get currentAsset())
   "SetText (Text)"(
    execute=@n1.Is Not Valid
    self=Get txtAsset() -> (txtAsset:Common Text Object Reference)
    InText="LOCTEXT(\"No valid asset assigned.  Right click an asset and choose: ScriptedActions -> EditMetadata \")"
   )
   > SetColorAndOpacity!(
    execute
    self=Get txtAsset()
    InColorAndOpacity_SpecifiedColor:Linear Color Structure=(R=0.9427,G=0.0578,B=0.0391,A=1)
    InColorAndOpacity_ColorUseRule:ESlateColorStylingMode Enum
   ) -> (then)
   > SetVisibility(self=Get btnFindAsset() -> (btnFindAsset:TA Generic 01 Button Object Reference), InVisibility=Collapsed)
   n2 = GetDisplayName(Object=Get currentAsset())
   "SetText (Text)"(execute=@n1.Is Valid, self=Get txtAsset(), InText="To Text (String)"(InString=Append(A="Viewing: ", B=@n2)))
   > SetColorAndOpacity(self=Get txtAsset(), InColorAndOpacity_SpecifiedColor=(R=0.9427,G=0.9427,B=0.9427,A=1))
   > SetVisibility(self=Get btnFindAsset(), InVisibility=Visible)
  ]}
]
variables [currentAsset {type=Object, default=null}]
widgets {
 info {rootWidgetClass=&CanvasPanel, widgetCount=9}
 widgets [
  CanvasPanel_0 {widgetClassPath=&CanvasPanel, widget=CanvasPanel_0}
  VerticalBox_103 {
   widgetClassPath=&VerticalBox
   parent=CanvasPanel_0
   widget=VerticalBox_103
   slot {class=CanvasPanelSlot, properties {layoutData {alignment=<0 0>, anchors {maximum=<1 1>, minimum=<0 0>}, offsets=<ltrb 0 0 0 0>}}}
  }
  Border_1 {
   widgetClassPath=&Border
   parent=VerticalBox_103
   widget=Border_1
   properties {brushColor=<rgba 0 0 0 0.5>, padding=<ltrb 5 5 5 5>}
   slot {class=VerticalBoxSlot}
  }
  HorizontalBox_0 {
   widgetClassPath=&HorizontalBox
   parent=Border_1
   widget=HorizontalBox_0
   slot {class=BorderSlot, properties {padding=<ltrb 5 5 5 5>}}
  }
  ScaleBox_0 {
   widgetClassPath=&ScaleBox
   parent=HorizontalBox_0
   widget=ScaleBox_0
   properties {stretchDirection=DownOnly}
   slot {class=HorizontalBoxSlot, properties {verticalAlignment=VAlign_Center, size {sizeRule=Fill, value=1}}}
  }
  txtAsset {
   widgetClassPath=&CommonTextBlock
   bIsVariable=true
   parent=ScaleBox_0
   widget=txtAsset
   properties {
    text="No valid asset assigned.  Right click an asset and choose: ScriptedActions -> EditMetadata "
    font {
     bForceMonospaced=false
     bMaterialIsStencil=false
     fontMaterial=null
     fontObject=&NotoSans
     letterSpacing=0
     monospacedWidth=1
     size=16
     skewAmount=0
     typefaceFontName=NotoSans-Regular
     outlineSettings {
      bApplyOutlineToDropShadows=false
      bMiteredCorners=false
      bSeparateFillAlpha=false
      outlineMaterial=null
      outlineSize=0
      outlineColor=<rgba 0 0 0 1>
     }
    }
    shadowOffset=<0 0>
   }
   slot {class=ScaleBoxSlot}
  }
  btnFindAsset {
   widgetClassPath=&TA_Generic01_Button
   parent=HorizontalBox_0
   widget=btnFindAsset
   properties {
    toolTipText=Go to this asset in the content browser.
    iconBrush {
     bIsDynamicallyLoaded=false
     drawAs=Image
     imageType=NoImage
     mirroring=NoMirror
     resourceName=None
     resourceObject=/Game/UI/Foundation/Textures/Icons/HUD/T-IconMagnifying-Glass-32
     tiling=NoTile
     imageSize=<24 24>
     margin=<ltrb 0 0 0 0>
     outlineSettings {
      bUseBrushTransparency=false
      roundingType=HalfHeightRadius
      width=0
      color {colorUseRule=UseColor_Specified, specifiedColor=<rgba 0 0 0 0>}
      cornerRadii=<0 0 0 1>
     }
     tintColor {colorUseRule=UseColor_Specified, specifiedColor=<rgba 1 1 1 1>}
     uVRegion {bIsValid=false, max=<0 0>, min=<0 0>}
    }
   }
   slot {class=HorizontalBoxSlot}
  }
  Border_0 {
   widgetClassPath=&Border
   parent=VerticalBox_103
   widget=Border_0
   properties {brushColor=<rgba 1 1 1 0>, padding=<ltrb 5 5 5 5>}
   slot {class=VerticalBoxSlot, properties {padding=<ltrb 15 15 15 15>, size {sizeRule=Fill, value=1}}}
  }
  Generic01_TabContainer {
   widgetClassPath=/Game/Blueprints/Blutilities/ForTechArt/Widgets/TabContainer/Generic01_TabContainer
   bIsVariable=true
   parent=Border_0
   widget=Generic01_TabContainer
   properties {initialTabs {
     Metadata Editor=&/EditAssetMetadata_EWidget.EditAssetMetadata_EWidget_C
     Tag Viewer=&/ViewAssetData_EWidget.ViewAssetData_EWidget_C
    }}
   slot {class=BorderSlot, properties {padding=<ltrb 5 5 5 5>}}
  }
 ]
}
